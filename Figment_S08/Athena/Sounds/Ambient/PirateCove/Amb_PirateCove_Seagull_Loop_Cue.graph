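# Looping ambient sound cue for seagulls at Pirate Cove, in the outdoor ambient sound class.
# A looping wave player feeds an indefinitely looping Enveloper, whose volume curve dips and returns to full within each cycle, so the gull bed swells and recedes over time.
# Overridden attenuation gives it long-range falloff, wide stereo spread and a fixed manual reverb send.
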SoundCue /Figment_S08/Athena/Sounds/Ambient/PirateCove/Amb_PirateCove_Seagull_Loop_Cue
root=Enveloper_0
nodes [
 {
  id=Enveloper_0
  children [WavePlayer_0]
  properties {
   bLoop=true
   bLoopIndefinitely=true
   loopEnd=70
   pitchCurve {editorCurveData {keys [{time=0}]}}
   volumeCurve {editorCurveData {keys [<at 0 1>,<at 15 0.5>,<at 25 0.2>,<at 40 0.2>,<at 50 1>]}}
  }
 }
 {id=WavePlayer_0, properties {bLooping=true, soundWaveAssetPtr=/Game/Athena/Sounds/Ambient/PirateCove/Amb_PirateCove_Seagull_Loop_01}}
]
properties {
 bOverrideAttenuation=true
 maxDistance=12000
 soundClassObject=&Fort_Ambient_Outdoors_Athena
 volumeMultiplier=1
 attenuationOverrides {falloffDistance=12000, manualReverbSendLevel=25, reverbSendMethod=Manual, stereoSpread=2048, attenuationShapeExtents {x=0}}
}
